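# Editor utility widget for rigging Sprout cosmetic vehicle skeletal meshes, with undo transactions around its actions.
# Selecting a mesh can auto-load its Atom model and wheel/steering meshes by naming convention.
# Place Bones uses mesh bounds and SkeletonModifier to position and commit bones.
# Place Sockets moves existing seat and exhaust sockets to Atom part transforms, ordered by embedded Python; Create Or Update Pie Files appears unfinished, with a broken cast.
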EditorUtilityWidgetBlueprint /AnimInEdTools/Vehicles/EUW_SproutCosmeticVehicle
graphs [
 Create Or Update Pie Files {nodes [
   n1 = LoadAsset_Blocking()
   > PrintString(InString=Get Class Display Name(Class=Get Class(Object=@n1) -> (ReturnValue:Object Class Reference)))
   > n2 = Bad cast node!(execute, Object=@n1) -> (then, CastFailed, AsSprout Convertible Vehicle:Object Reference) {is=title K2Node_DynamicCast}
   > PrintString(InString=GetDisplayName(Object=@n2))
   PrintString(execute=@n2.CastFailed, InString="Aaaaw shucks!")
   n3 = Replace(SourceString=GetObjectName(Object=Get Skeletal Mesh() -> (Skeletal Mesh:Skeletal Mesh Object Reference)), From=SKM_)
   n4 = LoadAsset(AssetPath=Append(A=Get Path() -> (Path:String), B=Append!(A=BP_, B=@n3, C:String=_Cosmetic) -> (ReturnValue)))
   > PrintString(InString=Append!(
     A=GetObjectName(Object=@n4)
     B="    "
     C=Get Object Path String(Object=@n4)
     D:String="    "
     E:String=GetSystemPath!(Object:Object Reference=@n4) -> (ReturnValue:String)
     F:String="   "
     G:String
    ) -> (ReturnValue))
   Set SKM(SKM:Skeletal Mesh Object Reference) -> (Output_Get:Skeletal Mesh Object Reference)
   > n5 = Branch()
   > PrintString(InString=Success)
   PrintString(execute=@n5.else, InString=Failure)
   LoadAsset(AssetPath=/JunoVehicleCosmetics/Bodies/Hero/BP_Hero_Cosmetic)
   CopyFile!(execute, InSourceFile:String, InDestFile:String, bSilent:bool=false) -> (then, ReturnValue:bool)
   PrintString(InString=Success)
   Create Or Update Pie Files() {is=title K2Node_FunctionEntry}
  ]}
 EventGraph {nodes [
   BeginTransaction()
   > Create Or Update Pie Files() {is=title K2Node_CallFunction}
   > EndTransaction()
   "On Clicked (Button_Place_Bones)"() {is=title K2Node_ComponentBoundEvent}
   > BeginTransaction()
   > Place Bones()
   > EndTransaction()
   n1 = "On Property Changed (DVCarSkeletalMesh)"!() -> (OutputDelegate, then, PropertyName:Name=None) {is=title K2Node_ComponentBoundEvent}
   > Branch(Condition="Equal (Name)"(A=@n1.PropertyName, B=Skeletal Mesh))
   > Update Arguments()
   "On Clicked (Button_Place_Sockets)"() {is=title K2Node_ComponentBoundEvent}
   > BeginTransaction()
   > Place Sockets()
   > EndTransaction()
   Event PreConstruct()
   > SetObject(
    self=[
     Get DVCarSkeletalMesh() -> (DVCarSkeletalMesh:Details View Object Reference)
     Get DVWheelStaticMeshes() -> (DVWheelStaticMeshes:Details View Object Reference)
     Get DVWheelBoneNames() -> (DVWheelBoneNames:Details View Object Reference)
     Get DVOptionalCarBones() -> (DVOptionalCarBones:Details View Object Reference)
     Get DVCarAtom() -> (DVCarAtom:Details View Object Reference)
     Get DVPiePath() -> (DVPiePath:Details View Object Reference)
    ]
    NewObject=Self-Reference()
   )
   Event Construct()
   Event Tick()
  ]}
 Place Bones {nodes [
   Place Bones() {is=title K2Node_FunctionEntry}
   > Clear(TargetArray:Array of Names=Get All Bone Names() -> (All Bone Names:Array of Names))
   > Clear(TargetArray:Array of Static Mesh Object References=Get Wheel Static Meshes() -> (Wheel Static Meshes:Array of Static Mesh Object References))
   > Clear(TargetArray:Array of Box Structures=Get Wheel Bounding Boxes() -> (Wheel Bounding Boxes:Array of Box Structures))
   n1 = Get Skeletal Mesh()
   > n2 = Branch(Condition=IsValid(Object=@n1))
   > n3 = Construct Object from Class(Class=/Script/SkeletalMeshModifiers.SkeletonModifier) -> (ReturnValue:Skeleton Modifier Object Reference)
   > SetSkeletalMesh!(execute, self:Skeleton Modifier Object Reference=@n3, InSkeletalMesh:Skeletal Mesh Object Reference=@n1) -> (then, ReturnValue:bool)
   > n4 = Set Skin Mod(Skin Mod:Skeleton Modifier Object Reference=@n3) -> (Output_Get:Skeleton Modifier Object Reference)
   > Set All Bone Names(All Bone Names:Array of Names=GetAllBoneNames!(self:Skeleton Modifier Object Reference=@n4) -> (ReturnValue:Array of Names)) -> (Output_Get:Array of Names)
   > n5 = Sequence!(execute) -> (then_0, then_1, then_2:Exec, then_3:Exec, then_4:Exec, then_5:Exec, then_6:Exec)
   > Set Wheel Bone Names(Wheel Bone Names:Array of Names=Make Array!(
     "[0]":Name=Get Bone Front L() -> (Bone Front L:Name=None)
     "[1]":Name=Get Bone Front R() -> (Bone Front R:Name=None)
     "[2]":Name=Get Bone Rear L() -> (Bone Rear L:Name=None)
     "[3]":Name=Get Bone Rear R() -> (Bone Rear R:Name=None)
    ) -> (Array:Array of Names)) -> (Output_Get:Array of Names)
   > n6 = Set Wheel Static Meshes(Wheel Static Meshes:Array of Static Mesh Object References=Make Array(
     "[0]":Static Mesh Object Reference=Get Wheel Front L Mesh() -> (Wheel Front L Mesh:Static Mesh Object Reference)
     "[1]":Static Mesh Object Reference=Get Wheel Front R Mesh() -> (Wheel Front R Mesh:Static Mesh Object Reference)
     "[2]":Static Mesh Object Reference=Get Wheel Rear L Mesh() -> (Wheel Rear L Mesh:Static Mesh Object Reference)
     "[3]":Static Mesh Object Reference=Get Wheel Rear R Mesh() -> (Wheel Rear R Mesh:Static Mesh Object Reference)
    ) -> (Array:Array of Static Mesh Object References)) -> (Output_Get:Array of Static Mesh Object References)
   > n7 = For Each Loop(Array:Array of Static Mesh Object References=@n6) -> (Array Element:Static Mesh Object Reference)
   > n8 = Branch(Condition=IsValid(Object=@n7))
   n9 = Get Wheel Bounding Boxes()
   > Add(
    TargetArray:Array of Box Structures=@n9
    NewItem:"Box Structure (by ref)"=GetBoundingBox!(self:Static Mesh Object Reference=@n7) -> (ReturnValue:Box Structure)
   ) {is=title K2Node_CallArrayFunction}
   Add!(execute=@n8.else, TargetArray=@n9, NewItem_Min:Vector="0, 0, 0", NewItem_Max:Vector="0, 0, 0") -> (then, ReturnValue)
   n10 = Get Steering Wheel Mesh() -> (Steering Wheel Mesh:Static Mesh Object Reference)
   Branch(execute=@n7.Completed, Condition=IsValid(Object=@n10))
   > Set Steering bb(Steering bb:Box Structure=GetBoundingBox(self=@n10)) -> (Output_Get:Box Structure)
   n11 = Branch(execute=@n5.then_1, Condition=Contains Item(TargetArray=Get All Bone Names(), ItemToFind=Get Frame() -> (Frame:Name=None)))
   > Set sum(sum:double=0) -> (Output_Get:double)
   > n12 = For Each Loop(Array:Array of Box Structures=Get Wheel Bounding Boxes()) -> (Array Element:Box Structure)
   > Set sum(sum=float + float(
     A=Get sum() -> (sum:double)
     B="Get Center (Box)"!(InBox:"Box Structure (by ref)"=@n12) -> (ReturnValue_X:double, ReturnValue_Y:double, ReturnValue_Z:double) {is=title K2Node_CallFunction}.ReturnValue_Z
    ))
   n13 = GetBoxCenterSize!(Box:Box Structure=@n12) -> (Center:Vector, Dimensions:Vector)
   > Branch(Condition="float > float"(A=Vector Length(A=@n13.Dimensions)))
   n14 = Get sum()
   SetBoneTransform!(
    execute=@n12.Completed
    self:Skeleton Modifier Object Reference=Get Skin Mod() -> (Skin Mod:Skeleton Modifier Object Reference)
    InBoneName:Name=Get Frame() -> (Frame=None)
    InNewTransform_Location:Vector=MakeVector(Y=0, Z=float / float(A=@n14, B=4))
    InNewTransform_Rotation:Rotator="0, 0, 0"
    InNewTransform_Scale:Vector="1,1,1"
    bMoveChildren:bool=true
   ) -> (then, ReturnValue:bool)
   PrintString(execute=@n11.else, InString=No valid Frame Name Given)
   n15 = Get Body() -> (Body:Name=None)
   n16 = Branch(execute=@n5.then_2, Condition=Contains Item(TargetArray=Get All Bone Names(), ItemToFind=@n15))
   > SetBoneTransform(
    self=Get Skin Mod()
    InBoneName=@n15
    InNewTransform_Location="0, 0, 0"
    InNewTransform_Rotation="0, 0, 0"
    InNewTransform_Scale="1,1,1"
    bMoveChildren=true
   )
   PrintString(execute=@n16.else, InString=No valid Body Name Given)
   n17 = Branch(execute=@n5.then_3, Condition=Contains Item(TargetArray=Get All Bone Names(), ItemToFind=Get Steering Wheel() -> (Steering Wheel:Name=None)))
   n18 = Get Steering bb() -> (Steering bb:Box Structure)
   n19 = GetBoxCenterSize(Box=@n18)
   > Branch(Condition="float > float"(A=Vector Length(A=@n19.Dimensions)))
   > n20 = GetParentBone!(execute, SkeletalMesh:Skeletal Mesh Object Reference=Get Skeletal Mesh(), BoneName:Name=Get Steering Wheel() -> (Steering Wheel=None)) -> (then, Valid:bool, ReturnValue:Name=None)
   n21 = GetBoneTransform!(self:Skeleton Modifier Object Reference=Get Skin Mod(), InBoneName:Name=@n20.ReturnValue, bGlobal:bool=true) -> (ReturnValue_Location:Vector, ReturnValue_Rotation:Rotator, ReturnValue_Scale:Vector)
   n22 = "Get Center (Box)"!(InBox=@n18) -> (ReturnValue:Vector)
   > SetBoneTransform(
    self=Get Skin Mod()
    InBoneName=Get Steering Wheel() -> (Steering Wheel=None)
    InNewTransform_Location=vector + vector(A=vector - vector(A=@n22, B=@n21), B="4.6631,0,5.7546")
    InNewTransform_Rotation="55,0,0"
    InNewTransform_Scale="1,1,1"
    bMoveChildren=false
   )
   PrintString(execute=@n17.else, InString=No valid Steering Wheel Name Given)
   n23 = Sequence!(execute=@n5.then_4) -> (then_0, then_1)
   > n24 = Branch(Condition=Contains Item(TargetArray=Get All Bone Names(), ItemToFind=Get Front Differential() -> (Front Differential:Name=None)))
   > n25 = GetParentBone(SkeletalMesh=Get Skeletal Mesh(), BoneName=Get Front Differential() -> (Front Differential=None)) -> (ReturnValue=None)
   n26 = Get Skin Mod()
   n27 = Get Wheel Bounding Boxes()
   n28 = vector + vector(
    A="Get Center (Box)"(InBox="Get (a copy)"(Array:Array of Box Structures=@n27) -> (Output:Box Structure))
    B="Get Center (Box)"(InBox="Get (a copy)"(Array=@n27, Dimension 1=1))
   )
   > SetBoneTransform(
    self=Get Skin Mod()
    InBoneName=Get Front Differential() -> (Front Differential=None)
    InNewTransform_Location=vector - vector(A=vector * vector(A=@n28, B:float=0.5) -> (ReturnValue=~), B=GetBoneTransform(self=@n26, InBoneName=@n25.ReturnValue, bGlobal=true))
    InNewTransform_Rotation="0, 0, 0"
    InNewTransform_Scale="1,1,1"
    bMoveChildren=true
   )
   PrintString(execute=@n24.else, InString=No valid Front Differential Name Given)
   n29 = Branch(
    execute=@n23.then_1
    Condition=Contains Item(TargetArray=Get All Bone Names(), ItemToFind=Get Rear Differential() -> (Rear Differential:Name=None))
   )
   > n30 = GetParentBone(SkeletalMesh=Get Skeletal Mesh(), BoneName=Get Rear Differential() -> (Rear Differential=None)) -> (ReturnValue=None)
   n31 = Get Skin Mod()
   n32 = Get Wheel Bounding Boxes()
   n33 = vector + vector(A="Get Center (Box)"(InBox="Get (a copy)"(Array=@n32, Dimension 1=2)), B="Get Center (Box)"(InBox="Get (a copy)"(Array=@n32, Dimension 1=3)))
   > SetBoneTransform(
    self=Get Skin Mod()
    InBoneName=Get Rear Differential() -> (Rear Differential=None)
    InNewTransform_Location=vector - vector(A=vector * vector(A=@n33, B=0.5) -> (ReturnValue=~), B=GetBoneTransform(self=@n31, InBoneName=@n30.ReturnValue, bGlobal=true))
    InNewTransform_Rotation="0, 0, 0"
    InNewTransform_Scale="1,1,1"
    bMoveChildren=true
   )
   PrintString(execute=@n29.else, InString=No valid Rear Differential Name Given)
   n34 = For Each Loop(Exec=@n5.then_5, Array:Array of Names=Get Wheel Bone Names() -> (Wheel Bone Names:Array of Names)) -> (Array Element:Name)
   > Set Current Wheel(Current Wheel:Name=@n34) -> (Output_Get:Name=None)
   > Set Index(Index:int=@n34.Array Index) -> (Output_Get:int=0)
   > Set Current Wheel Pos(Current Wheel Pos:Vector="0, 0, 0") -> (Output_Get:Vector)
   > n35 = Branch(Condition=Contains Item(TargetArray=Get All Bone Names(), ItemToFind=@n34))
   > n36 = Sequence()
   n37 = "To Float (Integer)"(InInt=Get Index() -> (Index:int=0))
   > n38 = Branch(Condition="Equal (Float)"(A="Division (Whole and Remainder)"!(Dividend:double=@n37, Divisor:double=2) -> (Remainder:double, ReturnValue:int=0)))
   n39 = Get Wheel Bounding Boxes()
   n40 = Get Index() -> (Index=0)
   n41 = "Get Center (Box)"!(InBox="Get (a copy)"(Array=@n39, Dimension 1=@n40)) -> (ReturnValue_X:double, ReturnValue_Y:double, ReturnValue_Z:double)
   n42 = "Get (a copy)"!(Array=@n39, Dimension 1=@n40) -> (
    Output_Min_X:double
    Output_Min_Y:double
    Output_Min_Z:double
    Output_Max_X:double
    Output_Max_Y:double
    Output_Max_Z:double
    Output_IsValid:bool=~
   )
   > Set!(execute, Current Wheel Pos_X:double=@n41, Current Wheel Pos_Y:double=@n42.Output_Max_Y, Current Wheel Pos_Z:double=@n41.ReturnValue_Z) -> (then, Output_Get:Vector)
   Set(execute=@n38.else, Current Wheel Pos_X=@n41, Current Wheel Pos_Y=@n42.Output_Min_Y, Current Wheel Pos_Z=@n41.ReturnValue_Z)
   n43 = GetParentBone(execute=@n36.then_1, SkeletalMesh=Get Skeletal Mesh(), BoneName=Get Current Wheel() -> (Current Wheel:Name=None)) -> (ReturnValue=None)
   > n44 = GetParentBone(SkeletalMesh=Get Skeletal Mesh(), BoneName=@n43.ReturnValue) -> (ReturnValue=None)
   n45 = Get Skin Mod()
   > SetBoneTransform(
    self=Get Skin Mod()
    InBoneName=@n43.ReturnValue
    InNewTransform_Location=vector - vector(A=Get Current Wheel Pos() -> (Current Wheel Pos:Vector), B=GetBoneTransform(self=@n45, InBoneName=@n44.ReturnValue, bGlobal=true))
    InNewTransform_Rotation="0, 0, 0"
    InNewTransform_Scale="1,1,1"
    bMoveChildren=true
   )
   > SetBoneTransform(
    self=Get Skin Mod()
    InBoneName=Get Current Wheel() -> (Current Wheel=None)
    InNewTransform_Location="0, 0, 0"
    InNewTransform_Rotation="0, 0, 0"
    InNewTransform_Scale="1,1,1"
    bMoveChildren=true
   )
   PrintString(execute=@n35.else, InString=Append!(A="No valid wheel bone named ", B="To String (Name)"(InName=@n34)) -> (ReturnValue))
   CommitSkeletonToSkeletalMesh!(execute=@n5.then_6, self:Skeleton Modifier Object Reference=Get Skin Mod()) -> (then, ReturnValue:bool)
   PrintString(execute=@n2.else, InString=No valid Skeletal Mesh entered)
  ]}
 Place Sockets {nodes [
   Place Sockets() {is=title K2Node_FunctionEntry}
   > n1 = Sequence!(execute) -> (then_0, then_1, then_2:Exec, then_3:Exec, then_4:Exec, then_5:Exec)
   n2 = Get Lego Model() -> (Lego Model:Atom Model Object Reference)
   n3 = For Each Loop!(
    Exec=@n1.then_2
    Array:Array of Atom Model Part Instance Info Structures=GetParts!(PartsCollection:"Atom Model Parts Collection Structure (by ref)"=GetPartsCollection!(self:Atom Model Object Reference=@n2) -> (ReturnValue:Atom Model Parts Collection Structure)) -> (ReturnValue:Array of Atom Model Part Instance Info Structures)
   ) -> (
    LoopBody
    Array Element_PartGuid:Atom Model Part Guid Structure
    Array Element_PartId:int
    Array Element_Primitive:Atom Primitive Object Reference
    Array Element_Transforms:Array of Transforms
    Array Element_Decorations:Array of Atom Model Part Decoration Info Structures
    Array Element_Colors:Array of Atom Model Part Color Info Structures
    Array Element_SelectionSets:Array of Strings
    Array Element_Group:String
    Array Element_ParentGroup:String
    Array Element_CommonParts:Array of Atom Common Part And Transform Structures
    Array Element_bIsUndersideVisible:bool=~
    Array Index
    Completed
   )
   > n4 = Branch(Condition=Contains(SearchIn=@n3.Array Element_Group, Substring=_seats))
   > n5 = For Each Loop!(Exec, Array:Array of Transforms=@n3.Array Element_Transforms) -> (LoopBody, Array Element_Location:Vector, Array Element_Rotation:Rotator, Array Element_Scale:Vector, Array Index, Completed)
   > Add!(
    execute
    TargetArray:Array of Vectors=Get Seats Socket Transforms() -> (Seats Socket Transforms:Array of Vectors)
    NewItem:"Vector (by ref)"=@n5
   ) -> (then, ReturnValue)
   n6 = Get Seat Counter() -> (Seat Counter:int=0)
   > Add!(
    execute
    TargetMap:Map of Integers to Vectors=Get Seats Transform Map() -> (Seats Transform Map:Map of Integers to Vectors)
    Key:"Integer (by ref)"=@n6
    Value:"Vector (by ref)"=@n5
   ) -> (then) {is=title K2Node_CallFunction}
   > Set Seat Counter(Seat Counter:int=int + int(A=1, B=@n6)) -> (Output_Get:int=0)
   Branch(execute=@n4.else, Condition=Contains(SearchIn=@n3.Array Element_Group, Substring=_boost))
   > n7 = For Each Loop(Array=@n3.Array Element_Transforms)
   > Add(TargetArray=Get Boost Socket Transforms() -> (Boost Socket Transforms:Array of Vectors), NewItem=@n7) {is=title K2Node_CallArrayFunction}
   n8 = Get Boost Counter() -> (Boost Counter:int=0)
   > Add(TargetMap=Get Boost Transform Map() -> (Boost Transform Map:Map of Integers to Vectors), Key=@n8, Value=@n7) {is=title K2Node_CallFunction}
   > Set Boost Counter(Boost Counter:int=int + int(A=1, B=@n8)) -> (Output_Get:int=0)
   n9 = Execute Python Script!(
    execute=@n1.then_3
    PythonScript="import unreal\r\n\r\nsorted_seats = []\r\nsorted_boosts = []\r\n\r\nfor ii, seat in enumerate(seat_transforms):\r\n    sorted_seats.append(((seat.x, seat.y, seat.z), ii))\r\n\r\nfor ii, boost in enumerate(boost_transforms):\r\n    sorted_boosts.append(((boost.x, boost.y, boost.z), ii))\r\n\r\n\r\nsorted_seats = sorted(sorted_seats, key=lambda item: (-item[0][0], item[0][1], -item[0][2]))\r\n\r\n\r\nif not right_hand_traffic:\r\n    print('RightHandTraffic')\r\n    sorted_seats = sorted(sorted_seats, key=lambda item: (-item[0][0], -item[0][1], -item[0][2]))\r\n\r\nsorted_boosts = sorted(sorted_boosts, key=lambda item: (-item[0][2], item[0][1], -item[0][0]))\r\n\r\nout_seat_order =   [item[1] for item in sorted_seats]\r\nout_boost_order = [item[1] for item in sorted_boosts]\r\n\r\nprint('Sorted Seat Transforms:')\r\nprint(sorted_seats)\r\nprint('')\r\nprint('Sorted Boost Transforms:')\r\nprint(sorted_boosts)\r\n\r\n"
    seat_transforms:Array of Vectors=Get Seats Socket Transforms()
    boost_transforms:Array of Vectors=Get Boost Socket Transforms()
    right_hand_traffic:bool=Get Right Hand Traffic() -> (Right Hand Traffic:bool)
   ) -> (then, ReturnValue, out_seat_order:Array of Integers, out_boost_order:Array of Integers) {inputs [seat_transforms, boost_transforms, right_hand_traffic], outputs [out_seat_order, out_boost_order]}
   > Set Sorted Seat Keys(Sorted Seat Keys:Array of Integers=@n9.out_seat_order) -> (Output_Get:Array of Integers)
   > Set Sorted Boost Keys(Sorted Boost Keys:Array of Integers=@n9.out_boost_order) -> (Output_Get:Array of Integers)
   Set Driver Socket Names(
    execute=@n1.then_0
    Driver Socket Names:Array of Names=Make Array!("[0]":Name=Driver, "[1]":Name=Driver_0, "[2]":Name=DriverExit, "[3]":Name=DriverChoice, "[4]":Name=DriverIndicator) -> (Array:Array of Names)
   ) -> (Output_Get:Array of Names)
   > Set Passenger Socket Names(Passenger Socket Names:Array of Names=Make Array("[0]"=Passenger, "[1]"=PassengerExit, "[2]"=PassengerChoice, "[3]"=PassengerIndicator, "[4]"=PassengerDriveBy)) -> (Output_Get:Array of Names)
   > Set Passenger B L Socket Names(Passenger B L Socket Names:Array of Names=Make Array!(
     "[0]"=PassengerBL
     "[1]"=PassengerBLIndicator_0
     "[2]"=PassengerBLChoice
     "[3]"=PassengerBLIndicator
     "[4]"=PassengerBLDriveBy
     "[5]":Name=PassengerExit_L
    ) -> (Array)) -> (Output_Get:Array of Names)
   > Set Passenger B R Socket Names(Passenger B R Socket Names:Array of Names=Make Array!("[0]"=PassengerBR, "[1]"=PassengerBRChoice, "[2]"=PassengerBRIndicator, "[3]"=PassengerBRDriveBy) -> (Array)) -> (Output_Get:Array of Names)
   > Set Boost Socket Names(Boost Socket Names:Array of Names=Make Array!("[0]"=FX_Exhaust_l, "[1]"=FX_Exhaust_r, "[2]"=FX_Exhaust_r_0) -> (Array)) -> (Output_Get:Array of Names)
   Clear(execute=@n1.then_1, TargetArray:Array of Vectors=Get Seats Socket Transforms())
   > Clear(TargetArray=Get Boost Socket Transforms())
   > Clear!(execute, TargetMap:Map of Integers to Vectors=Get Seats Transform Map()) -> (then) {is=title K2Node_CallFunction}
   > Clear(TargetMap=Get Boost Transform Map())
   > Clear(TargetArray:Array of Integers=Get Sorted Seat Keys() -> (Sorted Seat Keys:Array of Integers)) {is=title K2Node_CallArrayFunction}
   > Clear(TargetArray=Get Sorted Boost Keys() -> (Sorted Boost Keys:Array of Integers))
   > Set Seat Counter(Seat Counter=0) -> (Output_Get=0)
   > Set Boost Counter(Boost Counter=0) -> (Output_Get=0)
   > Clear(TargetArray:Array of Names=Get All Socket Names() -> (All Socket Names:Array of Names))
   n10 = Get Skeletal Mesh()
   n11 = For Loop(execute=@n1.then_4, LastIndex=NumSockets!(self:Skeletal Mesh Object Reference=@n10) -> (ReturnValue:int=0))
   > Add(
    TargetArray:Array of Names=Get All Socket Names()
    NewItem:"Name (by ref)"=Get SocketName!(self:Skeletal Mesh Socket Object Reference=GetSocketByIndex!(self:Skeletal Mesh Object Reference=@n10, Index:int=@n11) -> (ReturnValue:Skeletal Mesh Socket Object Reference)) -> (SocketName:Name=None)
   ) {is=title K2Node_CallArrayFunction}
   SetSkeletalMesh(execute=@n1.then_5, self=Get Skel Modifier() -> (Skel Modifier:Skeleton Modifier Object Reference), InSkeletalMesh=Get Skeletal Mesh())
   > n12 = Sequence!(execute) -> (then_0, then_1)
   > n13 = For Each Loop!(Exec, Array:Array of Integers=Get Sorted Seat Keys()) -> (LoopBody, Array Element:int, Array Index, Completed)
   > n14 = GetEditorProperty!(
    execute:Exec
    Object:Object Reference=Self-Reference()
    PropertyName:Name="Get (a copy)"!(
     Array:Array of Names=Make Array!("[0]"=Driver Socket Names, "[1]"=Passenger Socket Names, "[2]"=Passenger B L Socket Names, "[3]":Name=Passenger B R Socket Names) -> (Array)
     Dimension 1=@n13.Array Index
    ) -> (Output:Name)
   ) -> (then:Exec, PropertyValue:Array of Names, ReturnValue:bool)
   > n15 = For Each Loop(Array:Array of Names=@n14) -> (Array Element:Name)
   n16 = Get All Socket Names()
   > Branch(Condition="Not Equal (Integer)"(A=Find Item!(TargetArray:Array of Names=@n16, ItemToFind:"Name (by ref)"=@n15) -> (ReturnValue), B=-1))
   n17 = FindSocket!(self:Skinned Asset Object Reference=Get Skeletal Mesh(), InSocketName:Name=@n15) -> (ReturnValue:Skeletal Mesh Socket Object Reference)
   > n18 = Set Socket(Socket:Skeletal Mesh Socket Object Reference=@n17) -> (Output_Get:Skeletal Mesh Socket Object Reference)
   n19 = vector - vector(
    A=Find(TargetMap:Map of Integers to Vectors=Get Seats Transform Map(), Key:"Integer (by ref)"=@n13) -> (Value:Vector)
    B=GetBoneTransform(self=Get Skel Modifier(), InBoneName=Get BoneName!(self:Skeletal Mesh Socket Object Reference=@n18) -> (BoneName:Name=None), bGlobal=true)
   )
   > SetSocketLocalTransform!(
    execute
    self:Skeletal Mesh Socket Object Reference=@n17
    InTransform_Location:Vector=@n19
    InTransform_Rotation:Rotator="0, 0, 0"
    InTransform_Scale:Vector="1,1,1"
   ) -> (then)
   > PrintString(InString=Append!(
     A=Seat Sockets
     B
     C:String="To String (Name)"(InName=@n15)
     D:String
     E:String="To String (Vector)"!(InVec:Vector=@n19) -> (ReturnValue:String)
    ) -> (ReturnValue))
   n20 = For Each Loop(Exec=@n12.then_1, Array:Array of Integers=Get Sorted Boost Keys()) -> (Array Element:int)
   > n21 = For Each Loop(Array:Array of Names=Get Boost Socket Names() -> (Boost Socket Names:Array of Names)) -> (Array Element:Name)
   n22 = Get All Socket Names()
   > Branch(Condition="Not Equal (Integer)"(A=Find Item(TargetArray=@n22, ItemToFind=@n21), B=-1))
   n23 = FindSocket(self=Get Skeletal Mesh(), InSocketName=@n21)
   > n24 = Set Socket(Socket=@n23)
   n25 = vector - vector(
    A=Find(TargetMap=Get Boost Transform Map(), Key=@n20)
    B=GetBoneTransform(self=Get Skel Modifier(), InBoneName=Get BoneName(self=@n24) -> (BoneName=None), bGlobal=true)
   )
   > SetSocketLocalTransform(self=@n23, InTransform_Location=@n25, InTransform_Rotation="0, 0, 0", InTransform_Scale="1,1,1")
   > PrintString(InString=Append(A=Seat Sockets, C="To String (Name)"(InName=@n21), E="To String (Vector)"(InVec=@n25)))
  ]}
 Update Arguments {nodes [
   Update Arguments() {is=title K2Node_FunctionEntry}
   > Branch(Condition=Get Auto Update Args() -> (Auto Update Args:bool))
   > Branch(Condition=IsValid(Object=Get Skeletal Mesh()))
   > Set Mesh Variable Names(Mesh Variable Names:Array of Names=Make Array!("[0]"=Wheel Front L Mesh, "[1]"=Wheel Front R Mesh, "[2]"=Wheel Rear L Mesh, "[3]"=Wheel Rear R Mesh, "[4]":Name=Steering Wheel Mesh) -> (Array)) -> (Output_Get:Array of Names)
   > Set Bone Names(Bone Names:Array of Names=Make Array(
     "[0]"=Get Bone Front L() -> (Bone Front L=None)
     "[1]"=Get Bone Front R() -> (Bone Front R=None)
     "[2]"=Get Bone Rear L() -> (Bone Rear L=None)
     "[3]"=Get Bone Rear R() -> (Bone Rear R=None)
     "[4]"=Get Steering Wheel() -> (Steering Wheel=None)
    )) -> (Output_Get:Array of Names)
   > n1 = Sequence!(execute) -> (then_0, then_1, then_2:Exec)
   n2 = Get Skeletal Mesh()
   > n3 = Set SKM Name(SKM Name:String=GetObjectName(Object=@n2)) -> (Output_Get:String)
   > Set Car Local Name(Car Local Name:String=Replace(SourceString=@n3, From=SKM_)) -> (Output_Get:String)
   n4 = Split(SourceString=Split(SourceString=Get Object Path String(Object=@n2), InStr=/, SearchDir=FromEnd), InStr=/, SearchDir=FromEnd)
   n5 = Append!(A=@n4, B=/) -> (ReturnValue)
   > Set Car Path(Car Path:String=@n5) -> (Output_Get:String)
   > Set Mesh Path(Mesh Path:String=Append!(A=@n4, B=/, C:String=Mesh/) -> (ReturnValue)) -> (Output_Get:String)
   > Set Atom Path(Atom Path:String=Append(A=@n4, B=/, C=Atom/)) -> (Output_Get:String)
   > Set Path(Path:String=@n5) -> (Output_Get:String)
   n6 = Append!(A=Get Atom Path() -> (Atom Path:String), B=Get Car Local Name() -> (Car Local Name:String)) -> (ReturnValue)
   n7 = LoadAsset(execute=@n1.then_1, AssetPath=@n6)
   > n8 = Cast To AtomModel!(execute, Object=@n7) -> (then, CastFailed, AsAtom Model:Atom Model Object Reference) {is=title K2Node_DynamicCast}
   > Set Lego Model(Lego Model:Atom Model Object Reference=@n8) -> (Output_Get:Atom Model Object Reference)
   PrintString(execute=@n8.CastFailed, InString=Append(A="No Valid Atom Model Found In:  ", B=@n6))
   n9 = For Each Loop(Exec=@n1.then_2, Array=Get Mesh Variable Names() -> (Mesh Variable Names:Array of Names))
   n10 = Get Bone Names() -> (Bone Names:Array of Names)
   n11 = Append!(
    A=Get Mesh Path() -> (Mesh Path:String)
    B=SM_
    C:String=Get Car Local Name()
    D:String=_
    E:String="To String (Name)"(InName="Get (a copy)"(Array=@n10, Dimension 1=@n9.Array Index))
   ) -> (ReturnValue)
   > n12 = LoadAsset(AssetPath=@n11)
   > n13 = Cast To StaticMesh(Object=@n12)
   > SetEditorProperty(Object=Self-Reference(), PropertyName=@n9, PropertyValue:Static Mesh Object Reference=@n13)
   PrintString(execute=@n13.CastFailed, InString=Append!(A="No Valid Static Mesh Found In:  ", B=@n11) -> (ReturnValue))
  ]}
]
variables [
 All Socket Names {type="array<string>", default []}
 Auto Update Args {type=boolean, default=true}
 Body {type=string, default=body}
 Bone Front L {type=string, default=wheel_tilted_fr_l}
 Bone Front R {type=string, default=wheel_tilted_fr_r}
 Bone Rear L {type=string, default=wheel_tilted_bk_l}
 Bone Rear R {type=string, default=wheel_tilted_bk_r}
 Boost Socket Names {type="array<string>", default []}
 Car Name {type=string, default=""}
 Driver Socket Names {type="array<string>", default []}
 Frame {type=string, default=frame}
 Front Differential {type=string, default=front_differential}
 Lego Model {type=AtomModel, default=null}
 Passenger B L Socket Names {type="array<string>", default []}
 Passenger B R Socket Names {type="array<string>", default []}
 Passenger Socket Names {type="array<string>", default []}
 Path {type=string, default=""}
 Rear Differential {type=string, default=rear_differential}
 Right Hand Traffic {type=boolean, default=true}
 Skeletal Mesh {type=SkeletalMesh, default=null}
 Steering Wheel {type=string, default=steering_wheel}
 Steering Wheel Mesh {type=StaticMesh, default=null}
 Vehicle Parent Path {type=string, default=""}
 Wheel Front L Mesh {type=StaticMesh, default=null}
 Wheel Front R Mesh {type=StaticMesh, default=null}
 Wheel Rear L Mesh {type=StaticMesh, default=null}
 Wheel Rear R Mesh {type=StaticMesh, default=null}
]
widgets {
 info {namedSlotCount=4, rootWidgetClass=&StackBox, widgetCount=26}
 widgets [
  StackBox_0 {widgetClassPath=&StackBox, widget=StackBox_0, properties {orientation=Orient_Vertical}}
  EditorUtilityExpandableArea {
   widgetClassPath=/Script/Blutility.EditorUtilityExpandableArea
   parent=StackBox_0
   widget=EditorUtilityExpandableArea
   properties {
    bIsExpanded=true
    borderBrush {
     bIsDynamicallyLoaded=false
     drawAs=RoundedBox
     imageType=NoImage
     mirroring=NoMirror
     resourceName=None
     resourceObject=null
     tiling=NoTile
     imageSize=<0 0>
     margin=<ltrb 0 0 0 0>
     outlineSettings {
      bUseBrushTransparency=false
      roundingType=FixedRadius
      width=0
      color {colorUseRule=UseColor_Specified, specifiedColor=<rgba 0 0 0 0>}
      cornerRadii=<4 4 4 4>
     }
     tintColor {colorUseRule=UseColor_Specified, specifiedColor=<rgba 0 0 0 1>}
     uVRegion {bIsValid=false, max=<0 0>, min=<0 0>}
    }
   }
   slot {class=StackBoxSlot}
  }
  TextBlockBone_1 {
   widgetClassPath=&TextBlock
   namedSlotHost=EditorUtilityExpandableArea
   widget=TextBlockBone_1
   properties {
    text="Asset File Creation (Atom, SKM, SM)"
    font {
     bForceMonospaced=false
     bMaterialIsStencil=false
     fontMaterial=null
     fontObject=&Roboto
     letterSpacing=0
     monospacedWidth=1
     size=13.5
     skewAmount=0
     typefaceFontName=Italic
     outlineSettings {
      bApplyOutlineToDropShadows=false
      bMiteredCorners=false
      bSeparateFillAlpha=false
      outlineMaterial=null
      outlineSize=0
      outlineColor=<rgba 0 0 0 1>
     }
    }
   }
  }
  VerticalBox {widgetClassPath=&VerticalBox, namedSlotHost=EditorUtilityExpandableArea, widget=VerticalBox}
  HorizontalBox_75 {widgetClassPath=&HorizontalBox, parent=VerticalBox, widget=HorizontalBox_75, slot {class=VerticalBoxSlot}}
  EditorUtilityExpandableArea_287 {
   widgetClassPath=&EditorUtilityExpandableArea
   parent=StackBox_0
   widget=EditorUtilityExpandableArea_287
   properties {
    bIsExpanded=true
    borderBrush {
     bIsDynamicallyLoaded=false
     drawAs=RoundedBox
     imageType=NoImage
     mirroring=NoMirror
     resourceName=None
     resourceObject=null
     tiling=NoTile
     imageSize=<0 0>
     margin=<ltrb 0 0 0 0>
     outlineSettings {
      bUseBrushTransparency=false
      roundingType=FixedRadius
      width=0
      color {colorUseRule=UseColor_Specified, specifiedColor=<rgba 0 0 0 0>}
      cornerRadii=<4 4 4 4>
     }
     tintColor {colorUseRule=UseColor_Specified, specifiedColor=<rgba 0 0 0 1>}
     uVRegion {bIsValid=false, max=<0 0>, min=<0 0>}
    }
   }
   slot {class=StackBoxSlot, properties {size {sizeRule=Fill, value=1}}}
  }
  TextBlockBone {
   widgetClassPath=&TextBlock
   namedSlotHost=EditorUtilityExpandableArea_287
   widget=TextBlockBone
   properties {
    text=Bone Placement
    font {
     bForceMonospaced=false
     bMaterialIsStencil=false
     fontMaterial=null
     fontObject=&Roboto
     letterSpacing=0
     monospacedWidth=1
     size=13.5
     skewAmount=0
     typefaceFontName=Italic
     outlineSettings {
      bApplyOutlineToDropShadows=false
      bMiteredCorners=false
      bSeparateFillAlpha=false
      outlineMaterial=null
      outlineSize=0
      outlineColor=<rgba 0 0 0 1>
     }
    }
   }
  }
  VerticalBox_0 {widgetClassPath=&VerticalBox, namedSlotHost=EditorUtilityExpandableArea_287, widget=VerticalBox_0}
  HorizontalBoxSKM {widgetClassPath=&HorizontalBox, parent=VerticalBox_0, widget=HorizontalBoxSKM, slot {class=VerticalBoxSlot}}
  DVCarSkeletalMesh {
   widgetClassPath=&DetailsView
   parent=HorizontalBoxSKM
   widget=DVCarSkeletalMesh
   properties {categoriesToShow [Car Skeletal Mesh]}
   slot {class=HorizontalBoxSlot, properties {size {sizeRule=Fill, value=1}}}
  }
  HorizontalBoxWheelBones {widgetClassPath=&HorizontalBox, parent=VerticalBox_0, widget=HorizontalBoxWheelBones, slot {class=VerticalBoxSlot}}
  DVWheelBoneNames {
   widgetClassPath=&DetailsView
   parent=HorizontalBoxWheelBones
   widget=DVWheelBoneNames
   properties {categoriesToShow [Wheel Bone Names]}
   slot {class=HorizontalBoxSlot, properties {verticalAlignment=VAlign_Center, size {sizeRule=Fill, value=1}}}
  }
  HorizontalBoxOptionalBones {widgetClassPath=&HorizontalBox, parent=VerticalBox_0, widget=HorizontalBoxOptionalBones, slot {class=VerticalBoxSlot}}
  DVOptionalCarBones {
   widgetClassPath=&DetailsView
   parent=HorizontalBoxOptionalBones
   widget=DVOptionalCarBones
   properties {categoriesToShow [Optional Bone Names]}
   slot {class=HorizontalBoxSlot, properties {size {sizeRule=Fill, value=1}}}
  }
  HorizontalBoxWheelMeshes {widgetClassPath=&HorizontalBox, parent=VerticalBox_0, widget=HorizontalBoxWheelMeshes, slot {class=VerticalBoxSlot}}
  DVWheelStaticMeshes {
   widgetClassPath=&DetailsView
   parent=HorizontalBoxWheelMeshes
   widget=DVWheelStaticMeshes
   properties {categoriesToShow [Car Static Meshes]}
   slot {class=HorizontalBoxSlot, properties {size {sizeRule=Fill, value=1}}}
  }
  Button_Place_Bones {
   widgetClassPath=&Button
   parent=VerticalBox_0
   widget=Button_Place_Bones
   properties {widgetStyle {
     clickedSlateSound {resourceObject=null}
     disabled {
      bIsDynamicallyLoaded=false
      drawAs=RoundedBox
      imageType=NoImage
      mirroring=NoMirror
      resourceName=None
      resourceObject=null
      tiling=NoTile
      imageSize=<0 0>
      margin=<ltrb 0 0 0 0>
      outlineSettings {
       bUseBrushTransparency=false
       roundingType=FixedRadius
       width=1
       color {colorUseRule=UseColor_ColorTable, specifiedColor=<rgba 0 0 0 0>}
       cornerRadii=<4 4 4 4>
      }
      tintColor {colorUseRule=UseColor_Specified, specifiedColor=<rgba 0.0395 0.0395 0.0395 1>}
      uVRegion {bIsValid=false, max=<0 0>, min=<0 0>}
     }
     disabledForeground {colorUseRule=UseColor_ColorTable, specifiedColor=<rgba 0 0 0 0>}
     hovered {
      bIsDynamicallyLoaded=false
      drawAs=RoundedBox
      imageType=NoImage
      mirroring=NoMirror
      resourceName=None
      resourceObject=null
      tiling=NoTile
      imageSize=<32 32>
      margin=<ltrb 0 0 0 0>
      outlineSettings {
       bUseBrushTransparency=false
       roundingType=FixedRadius
       width=1
       color {colorUseRule=UseColor_ColorTable, specifiedColor=<rgba 0 0 0 0>}
       cornerRadii=<4 4 4 4>
      }
      tintColor {colorUseRule=UseColor_Specified, specifiedColor=<rgba 0.0953 0.0953 0.0953 1>}
      uVRegion {bIsValid=false, max=<0 0>, min=<0 0>}
     }
     hoveredForeground {colorUseRule=UseColor_ColorTable, specifiedColor=<rgba 0 0 0 0>}
     hoveredSlateSound {resourceObject=null}
     normal {
      bIsDynamicallyLoaded=false
      drawAs=RoundedBox
      imageType=NoImage
      mirroring=NoMirror
      resourceName=None
      resourceObject=null
      tiling=NoTile
      imageSize=<32 32>
      margin=<ltrb 0 0 0 0>
      outlineSettings {
       bUseBrushTransparency=false
       roundingType=FixedRadius
       width=1
       color {colorUseRule=UseColor_ColorTable, specifiedColor=<rgba 0 0 0 0>}
       cornerRadii=<4 4 4 4>
      }
      tintColor {colorUseRule=UseColor_Specified, specifiedColor=<rgba 0.0395 0.0395 0.0395 1>}
      uVRegion {bIsValid=false, max=<0 0>, min=<0 0>}
     }
     normalForeground {colorUseRule=UseColor_ColorTable, specifiedColor=<rgba 0 0 0 0>}
     normalPadding=<ltrb 12 1.5 12 1.5>
     pressed {
      bIsDynamicallyLoaded=false
      drawAs=RoundedBox
      imageType=NoImage
      mirroring=NoMirror
      resourceName=None
      resourceObject=null
      tiling=NoTile
      imageSize=<32 32>
      margin=<ltrb 0 0 0 0>
      outlineSettings {
       bUseBrushTransparency=false
       roundingType=FixedRadius
       width=1
       color {colorUseRule=UseColor_ColorTable, specifiedColor=<rgba 0 0 0 0>}
       cornerRadii=<4 4 4 4>
      }
      tintColor {colorUseRule=UseColor_Specified, specifiedColor=<rgba 0.0284 0.0284 0.0284 1>}
      uVRegion {bIsValid=false, max=<0 0>, min=<0 0>}
     }
     pressedForeground {colorUseRule=UseColor_ColorTable, specifiedColor=<rgba 0 0 0 0>}
     pressedPadding=<ltrb 12 2.5 12 0.5>
     pressedSlateSound {resourceObject=null}
    }}
   slot {class=VerticalBoxSlot, properties {verticalAlignment=VAlign_Center}}
  }
  TextBlock_1 {
   widgetClassPath=&TextBlock
   parent=Button_Place_Bones
   widget=TextBlock_1
   properties {
    autoWrapText=true
    justification=Center
    text=Place Bones
    colorAndOpacity {colorUseRule=UseColor_Specified, specifiedColor=<rgba 0.5255 0.5255 0.5255 1>}
    font {
     bForceMonospaced=false
     bMaterialIsStencil=false
     fontMaterial=null
     fontObject=&Roboto
     letterSpacing=0
     monospacedWidth=1
     size=12
     skewAmount=0
     typefaceFontName=Bold
     outlineSettings {
      bApplyOutlineToDropShadows=false
      bMiteredCorners=false
      bSeparateFillAlpha=false
      outlineMaterial=null
      outlineSize=0
      outlineColor=<rgba 0 0 0 1>
     }
    }
   }
   slot {class=ButtonSlot, properties {horizontalAlignment=HAlign_Fill}}
  }
  Spacer_85 {widgetClassPath=&Spacer, parent=VerticalBox_0, widget=Spacer_85, properties {size=<1 24>}, slot {class=VerticalBoxSlot}}
  HorizontalBox_275 {widgetClassPath=&HorizontalBox, parent=VerticalBox_0, widget=HorizontalBox_275, slot {class=VerticalBoxSlot}}
  DVCarAtom {
   widgetClassPath=&DetailsView
   parent=HorizontalBox_275
   widget=DVCarAtom
   properties {categoriesToShow [Car Atom Model]}
   slot {class=HorizontalBoxSlot, properties {size {sizeRule=Fill, value=1}}}
  }
  Button_Place_Sockets {
   widgetClassPath=&Button
   parent=VerticalBox_0
   widget=Button_Place_Sockets
   properties {widgetStyle {
     clickedSlateSound {resourceObject=null}
     disabled {
      bIsDynamicallyLoaded=false
      drawAs=RoundedBox
      imageType=NoImage
      mirroring=NoMirror
      resourceName=None
      resourceObject=null
      tiling=NoTile
      imageSize=<0 0>
      margin=<ltrb 0 0 0 0>
      outlineSettings {
       bUseBrushTransparency=false
       roundingType=FixedRadius
       width=1
       color {colorUseRule=UseColor_ColorTable, specifiedColor=<rgba 0 0 0 0>}
       cornerRadii=<4 4 4 4>
      }
      tintColor {colorUseRule=UseColor_Specified, specifiedColor=<rgba 0.0395 0.0395 0.0395 1>}
      uVRegion {bIsValid=false, max=<0 0>, min=<0 0>}
     }
     disabledForeground {colorUseRule=UseColor_ColorTable, specifiedColor=<rgba 0 0 0 0>}
     hovered {
      bIsDynamicallyLoaded=false
      drawAs=RoundedBox
      imageType=NoImage
      mirroring=NoMirror
      resourceName=None
      resourceObject=null
      tiling=NoTile
      imageSize=<32 32>
      margin=<ltrb 0 0 0 0>
      outlineSettings {
       bUseBrushTransparency=false
       roundingType=FixedRadius
       width=1
       color {colorUseRule=UseColor_ColorTable, specifiedColor=<rgba 0 0 0 0>}
       cornerRadii=<4 4 4 4>
      }
      tintColor {colorUseRule=UseColor_Specified, specifiedColor=<rgba 0.0953 0.0953 0.0953 1>}
      uVRegion {bIsValid=false, max=<0 0>, min=<0 0>}
     }
     hoveredForeground {colorUseRule=UseColor_ColorTable, specifiedColor=<rgba 0 0 0 0>}
     hoveredSlateSound {resourceObject=null}
     normal {
      bIsDynamicallyLoaded=false
      drawAs=RoundedBox
      imageType=NoImage
      mirroring=NoMirror
      resourceName=None
      resourceObject=null
      tiling=NoTile
      imageSize=<32 32>
      margin=<ltrb 0 0 0 0>
      outlineSettings {
       bUseBrushTransparency=false
       roundingType=FixedRadius
       width=1
       color {colorUseRule=UseColor_ColorTable, specifiedColor=<rgba 0 0 0 0>}
       cornerRadii=<4 4 4 4>
      }
      tintColor {colorUseRule=UseColor_Specified, specifiedColor=<rgba 0.0395 0.0395 0.0395 1>}
      uVRegion {bIsValid=false, max=<0 0>, min=<0 0>}
     }
     normalForeground {colorUseRule=UseColor_ColorTable, specifiedColor=<rgba 0 0 0 0>}
     normalPadding=<ltrb 12 1.5 12 1.5>
     pressed {
      bIsDynamicallyLoaded=false
      drawAs=RoundedBox
      imageType=NoImage
      mirroring=NoMirror
      resourceName=None
      resourceObject=null
      tiling=NoTile
      imageSize=<32 32>
      margin=<ltrb 0 0 0 0>
      outlineSettings {
       bUseBrushTransparency=false
       roundingType=FixedRadius
       width=1
       color {colorUseRule=UseColor_ColorTable, specifiedColor=<rgba 0 0 0 0>}
       cornerRadii=<4 4 4 4>
      }
      tintColor {colorUseRule=UseColor_Specified, specifiedColor=<rgba 0.0284 0.0284 0.0284 1>}
      uVRegion {bIsValid=false, max=<0 0>, min=<0 0>}
     }
     pressedForeground {colorUseRule=UseColor_ColorTable, specifiedColor=<rgba 0 0 0 0>}
     pressedPadding=<ltrb 12 2.5 12 0.5>
     pressedSlateSound {resourceObject=null}
    }}
   slot {class=VerticalBoxSlot}
  }
  TextBlock {
   widgetClassPath=&TextBlock
   parent=Button_Place_Sockets
   widget=TextBlock
   properties {
    autoWrapText=true
    justification=Center
    text=Place Sockets
    colorAndOpacity {colorUseRule=UseColor_Specified, specifiedColor=<rgba 0.5255 0.5255 0.5255 1>}
    font {
     bForceMonospaced=false
     bMaterialIsStencil=false
     fontMaterial=null
     fontObject=&Roboto
     letterSpacing=0
     monospacedWidth=1
     size=12
     skewAmount=0
     typefaceFontName=Bold
     outlineSettings {
      bApplyOutlineToDropShadows=false
      bMiteredCorners=false
      bSeparateFillAlpha=false
      outlineMaterial=null
      outlineSize=0
      outlineColor=<rgba 0 0 0 1>
     }
    }
   }
   slot {class=ButtonSlot, properties {horizontalAlignment=HAlign_Fill}}
  }
  Spacer {widgetClassPath=&Spacer, parent=VerticalBox_0, widget=Spacer, properties {size=<1 24>}, slot {class=VerticalBoxSlot}}
  HorizontalBox_PIE_File_Path {widgetClassPath=&HorizontalBox, parent=VerticalBox_0, widget=HorizontalBox_PIE_File_Path, slot {class=VerticalBoxSlot}}
  DVPiePath {
   widgetClassPath=&DetailsView
   parent=HorizontalBox_PIE_File_Path
   widget=DVPiePath
   properties {categoriesToShow [Pie Files Location]}
   slot {class=HorizontalBoxSlot, properties {size {sizeRule=Fill, value=1}}}
  }
 ]
}
